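# Masked, lit vine material for Niagara mesh particles in the OakWeld pickaxe effects.
# SplineThicken sets strand width, blending tip and middle widths with a sine profile that follows the dynamic Grow parameter, which also masks the strand along V.
# World position offset adds gravity and a second dynamic vector, which repeats Vector X on all axes, weighted by a power of inverted V to bend the vine.
# A cylinder normal map is sampled.
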Material /BRCosmetics/Weapons/FORT_Melee/Pickaxe_OakWeld/FX/M_Pickaxe_OakWeld_Vine
graph {
 properties {
  bAutomaticallySetUsageInEditor=false
  bTangentSpaceNormal=false
  bUsedWithNiagaraMeshParticles=true
  blendMode=BLEND_Masked
  usedShadingModels=Default Lit
 }
 nodes [
  n1 = DynamicParameter() {defaultValue=<rgba 1 0 1 1>, paramNames [Grow, Gravity Intensity, Width, Param4]}
  n2 = TextureCoordinate()
  n3 = Sine(None=SmoothStep(Min=OneMinus(None=@n1.Grow), Value=ComponentMask(None=@n2) {g=true}) {constMin=0.9}) {period=2}
  n4 = Multiply(
   A=DynamicParameter() {defaultValue=<rgba 1 0 1 1>, paramNames [Grow, Gravity Intensity, Width, Param4]}.Width
   B=LinearInterpolate(
    A=ScalarParameter() {parameterName=Spline Width Tips, sliderMax=10}
    B=ScalarParameter() {defaultValue=5, parameterName=Spline Width Mid, sliderMax=20}
    Alpha=Saturate(None=@n3)
   )
  )
  n5 = Multiply(
   A=DynamicParameter() {defaultValue=<rgba 1 0 1 1>, paramNames [Grow, Gravity Intensity, Width, Param4]}.Gravity Intensity
   B=VectorParameter() {parameterName=World Space Gravity, defaultValue=<rgba 0 0 -1 0>}.RGB
  )
  n6 = DynamicParameter() {parameterIndex=1, defaultValue=<rgba 1 1 1 0>, paramNames [Vector X, Vector Y, Vector Z, Intensity]}
  n7 = Multiply(
   A=MaterialFunctionCall(R=@n6.Vector X, G=@n6.Vector X, B=@n6.Vector X) {materialFunction=&AppendMany}.RGB
   B=@n6.Intensity
  )
  n8 = OneMinus(None=ComponentMask(None=TextureCoordinate()) {g=true})
  n9 = ScalarParameter() {defaultValue=2, parameterName=WPO Bend Power, sliderMax=10}
  Add(
   A=MaterialFunctionCall(WidthBase=@n4, WidthTip=@n4) {materialFunction=&SplineThicken}.WorldPositionOffset
   B=Multiply(A=Add(A=@n5, B=@n7), B=Power(Base=@n8, Exp=@n9) {constExponent=5})
  )
  n10 = OneMinus(None=DynamicParameter() {defaultValue=<rgba 1 0 1 1>, paramNames [Grow, Gravity Intensity, Width, Param4]}.Grow)
  n11 = TextureCoordinate()
  Multiply(A=ParticleColor().A, B=SmoothStep(Min=Add(A=@n10) {constB=-0.05}, Max=@n10, Value=ComponentMask(None=@n11) {g=true}) {constMin=0.9})
  TextureSample() {samplerType=SAMPLERTYPE_Normal, texture=/Game/Environments/Asteria/Foliage/Grass/Textures/Grass/CylinderNormal_N}
 ]
}
