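# Pickaxe_OakWeld axe Anim Blueprint that adds secondary physics motion to six leaf bones.
# The cached input pose feeds two Anim Dynamics chains: a looser Actor-space one with per-bone tuning and a stiffer, uniform World-space one.
# A bool blend with no driving variable leaves the Actor-space chain selected by default, then an additive rotation offsets Dyn_leaves_01 before output.
# The EventGraph holds no working logic.
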AnimBlueprint /BRCosmetics/Weapons/FORT_Melee/Pickaxe_OakWeld/Meshes/OakWeld_Axe_AnimBP
target_skeleton=&/OakWeld_Axe_Skeleton
graphs [
 AnimGraph {nodes [
   n1 = Anim Dynamics - Bone: Dyn_leaves_01(ComponentPose=Local To Component(LocalPose=Use cached pose 'Base'()), Alpha=1, ExternalForce="0,0,0") {
    angularDampingOverride=0.99
    angularSpringConstant=600
    bAngularSpring=true
    bGravityOverrideInSimSpace=true
    bOverrideAngularDamping=true
    bUseGravityOverride=true
    simulationSpace=Actor
    alphaBoolBlend {blendOption=Linear}
    boundBone {boneName=Dyn_leaves_01}
    componentAppliedLinearAccClamp=<500 500 500>
    componentLinearAccScale=<0.35 0.4 0.55>
    componentLinearVelScale=<0.65 0.64 0.55>
    gravityOverride {z=-980}
    physicsBodyDefinitions [{
      boundBone {boneName=Dyn_leaves_01}
      boxExtents=<10 10 10>
      constraintSetup {angularLimitsMax=<5 0 9>, angularLimitsMin=<-5 -25 -9>, angularTarget=<1 0 0>}
      localJointOffset=<0.35 0 0>
     }]
   }
   n2 = Anim Dynamics - Bone: Dyn_leaves_02(ComponentPose=@n1, Alpha=1, ExternalForce="0,0,0") {
    angularDampingOverride=0.9
    angularSpringConstant=650
    bAngularSpring=true
    bGravityOverrideInSimSpace=true
    bOverrideAngularDamping=true
    bUseGravityOverride=true
    simulationSpace=Actor
    alphaBoolBlend {blendOption=Linear}
    boundBone {boneName=Dyn_leaves_02}
    componentAppliedLinearAccClamp=<500 500 500>
    componentLinearAccScale=<0.4 0.4 0.6>
    componentLinearVelScale=<0.7 0.7 0.6>
    gravityOverride {z=-980}
    physicsBodyDefinitions [{
      boundBone {boneName=Dyn_leaves_02}
      boxExtents=<10 10 10>
      constraintSetup {angularLimitsMax=<5 0 9>, angularLimitsMin=<-5 -30 -9>, angularTarget=<1 0 0>}
      localJointOffset=<0.4 0 0>
     }]
   }
   n3 = Anim Dynamics - Bone: Dyn_leaves_03(ComponentPose=@n2, Alpha=1, ExternalForce="0,0,0") {
    angularDampingOverride=0.92
    angularSpringConstant=700
    bAngularSpring=true
    bGravityOverrideInSimSpace=true
    bOverrideAngularDamping=true
    bUseGravityOverride=true
    simulationSpace=Actor
    alphaBoolBlend {blendOption=Linear}
    boundBone {boneName=Dyn_leaves_03}
    componentAppliedLinearAccClamp=<500 500 500>
    componentLinearAccScale=<0.35 0.4 0.55>
    componentLinearVelScale=<0.65 0.64 0.55>
    gravityOverride {z=-980}
    physicsBodyDefinitions [{
      boundBone {boneName=Dyn_leaves_03}
      boxExtents=<10 10 10>
      constraintSetup {angularLimitsMax=<5 0 9>, angularLimitsMin=<-5 -25 -9>, angularTarget=<1 0 0>}
      localJointOffset=<0.35 0 0>
     }]
   }
   n4 = Anim Dynamics - Bone: Dyn_leaves_04(ComponentPose=@n3, Alpha=1, ExternalForce="0,0,0") {
    angularDampingOverride=0.95
    angularSpringConstant=850
    bAngularSpring=true
    bGravityOverrideInSimSpace=true
    bOverrideAngularDamping=true
    bUseGravityOverride=true
    simulationSpace=Actor
    alphaBoolBlend {blendOption=Linear}
    boundBone {boneName=Dyn_leaves_04}
    componentAppliedLinearAccClamp=<500 500 500>
    componentLinearAccScale=<0.4 0.4 0.6>
    componentLinearVelScale=<0.7 0.7 0.6>
    gravityOverride {z=-980}
    physicsBodyDefinitions [{
      boundBone {boneName=Dyn_leaves_04}
      boxExtents=<10 10 10>
      constraintSetup {angularLimitsMax=<5 0 9>, angularLimitsMin=<-5 -30 -9>, angularTarget=<1 0 0>}
      localJointOffset=<0.4 0 0>
     }]
   }
   n5 = Anim Dynamics - Bone: Dyn_leaves_05(ComponentPose=@n4, Alpha=1, ExternalForce="0,0,0") {
    angularDampingOverride=0.98
    angularSpringConstant=800
    bAngularSpring=true
    bGravityOverrideInSimSpace=true
    bOverrideAngularDamping=true
    bUseGravityOverride=true
    simulationSpace=Actor
    alphaBoolBlend {blendOption=Linear}
    boundBone {boneName=Dyn_leaves_05}
    componentAppliedLinearAccClamp=<500 500 500>
    componentLinearAccScale=<0.35 0.4 0.55>
    componentLinearVelScale=<0.65 0.64 0.55>
    gravityOverride {z=-980}
    physicsBodyDefinitions [{
      boundBone {boneName=Dyn_leaves_05}
      boxExtents=<10 10 10>
      constraintSetup {angularLimitsMax=<5 0 9>, angularLimitsMin=<-5 -25 -9>, angularTarget=<1 0 0>}
      localJointOffset=<0.35 0 0>
     }]
   }
   n6 = Anim Dynamics - Bone: Dyn_leaves_06(ComponentPose=@n5, Alpha=1, ExternalForce="0,0,0") {
    angularDampingOverride=0.87
    angularSpringConstant=750
    bAngularSpring=true
    bGravityOverrideInSimSpace=true
    bOverrideAngularDamping=true
    bUseGravityOverride=true
    simulationSpace=Actor
    alphaBoolBlend {blendOption=Linear}
    boundBone {boneName=Dyn_leaves_06}
    componentAppliedLinearAccClamp=<500 500 500>
    componentLinearAccScale=<0.4 0.4 0.6>
    componentLinearVelScale=<0.7 0.7 0.6>
    gravityOverride {z=-980}
    physicsBodyDefinitions [{
      boundBone {boneName=Dyn_leaves_06}
      boxExtents=<10 10 10>
      constraintSetup {angularLimitsMax=<5 0 9>, angularLimitsMin=<-5 -30 -9>, angularTarget=<1 0 0>}
      localJointOffset=<0.4 0 0>
     }]
   }
   n7 = Component To Local(ComponentPose=@n6)
   n8 = Anim Dynamics - Bone: Dyn_leaves_01(ComponentPose=Local To Component(LocalPose=Use cached pose 'Base'()), Alpha=1, ExternalForce="0,0,0") {
    angularDampingOverride=0.9999
    angularSpringConstant=1000
    bAngularSpring=true
    bOverrideAngularDamping=true
    bUseGravityOverride=true
    simulationSpace=World
    alphaBoolBlend {blendOption=Linear}
    boundBone {boneName=Dyn_leaves_01}
    componentAppliedLinearAccClamp=<500 500 500>
    componentLinearAccScale=<0.3 0.3 0.3>
    componentLinearVelScale=<0.3 0.3 0.3>
    gravityOverride {z=-980}
    physicsBodyDefinitions [{
      boundBone {boneName=Dyn_leaves_01}
      boxExtents=<10 10 10>
      constraintSetup {angularLimitsMax=<0 0 5>, angularLimitsMin=<0 -13 -5>, angularTarget=<1 0 0>}
      localJointOffset=<0.2 0 0>
     }]
   }
   n9 = Anim Dynamics - Bone: Dyn_leaves_02(ComponentPose=@n8, Alpha=1, ExternalForce="0,0,0") {
    angularDampingOverride=0.9999
    angularSpringConstant=1000
    bAngularSpring=true
    bOverrideAngularDamping=true
    bUseGravityOverride=true
    simulationSpace=World
    alphaBoolBlend {blendOption=Linear}
    boundBone {boneName=Dyn_leaves_02}
    componentAppliedLinearAccClamp=<500 500 500>
    componentLinearAccScale=<0.3 0.3 0.3>
    componentLinearVelScale=<0.3 0.3 0.3>
    gravityOverride {z=-980}
    physicsBodyDefinitions [{
      boundBone {boneName=Dyn_leaves_02}
      boxExtents=<10 10 10>
      constraintSetup {angularLimitsMax=<0 0 5>, angularLimitsMin=<0 -13 -5>, angularTarget=<1 0 0>}
      localJointOffset=<0.2 0 0>
     }]
   }
   n10 = Anim Dynamics - Bone: Dyn_leaves_03(ComponentPose=@n9, Alpha=1, ExternalForce="0,0,0") {
    angularDampingOverride=0.9999
    angularSpringConstant=1000
    bAngularSpring=true
    bOverrideAngularDamping=true
    bUseGravityOverride=true
    simulationSpace=World
    alphaBoolBlend {blendOption=Linear}
    boundBone {boneName=Dyn_leaves_03}
    componentAppliedLinearAccClamp=<500 500 500>
    componentLinearAccScale=<0.3 0.3 0.3>
    componentLinearVelScale=<0.3 0.3 0.3>
    gravityOverride {z=-980}
    physicsBodyDefinitions [{
      boundBone {boneName=Dyn_leaves_03}
      boxExtents=<10 10 10>
      constraintSetup {angularLimitsMax=<0 0 5>, angularLimitsMin=<0 -13 -5>, angularTarget=<1 0 0>}
      localJointOffset=<0.2 0 0>
     }]
   }
   n11 = Anim Dynamics - Bone: Dyn_leaves_04(ComponentPose=@n10, Alpha=1, ExternalForce="0,0,0") {
    angularDampingOverride=0.9999
    angularSpringConstant=1000
    bAngularSpring=true
    bOverrideAngularDamping=true
    bUseGravityOverride=true
    simulationSpace=World
    alphaBoolBlend {blendOption=Linear}
    boundBone {boneName=Dyn_leaves_04}
    componentAppliedLinearAccClamp=<500 500 500>
    componentLinearAccScale=<0.3 0.3 0.3>
    componentLinearVelScale=<0.3 0.3 0.3>
    gravityOverride {z=-980}
    physicsBodyDefinitions [{
      boundBone {boneName=Dyn_leaves_04}
      boxExtents=<10 10 10>
      constraintSetup {angularLimitsMax=<0 0 5>, angularLimitsMin=<0 -13 -5>, angularTarget=<1 0 0>}
      localJointOffset=<0.2 0 0>
     }]
   }
   n12 = Anim Dynamics - Bone: Dyn_leaves_05(ComponentPose=@n11, Alpha=1, ExternalForce="0,0,0") {
    angularDampingOverride=0.9999
    angularSpringConstant=1000
    bAngularSpring=true
    bOverrideAngularDamping=true
    bUseGravityOverride=true
    simulationSpace=World
    alphaBoolBlend {blendOption=Linear}
    boundBone {boneName=Dyn_leaves_05}
    componentAppliedLinearAccClamp=<500 500 500>
    componentLinearAccScale=<0.3 0.3 0.3>
    componentLinearVelScale=<0.3 0.3 0.3>
    gravityOverride {z=-980}
    physicsBodyDefinitions [{
      boundBone {boneName=Dyn_leaves_05}
      boxExtents=<10 10 10>
      constraintSetup {angularLimitsMax=<0 0 5>, angularLimitsMin=<0 -13 -5>, angularTarget=<1 0 0>}
      localJointOffset=<0.2 0 0>
     }]
   }
   n13 = Anim Dynamics - Bone: Dyn_leaves_06(ComponentPose=@n12, Alpha=1, ExternalForce="0,0,0") {
    angularDampingOverride=0.9999
    angularSpringConstant=1000
    bAngularSpring=true
    bOverrideAngularDamping=true
    bUseGravityOverride=true
    simulationSpace=World
    alphaBoolBlend {blendOption=Linear}
    boundBone {boneName=Dyn_leaves_06}
    componentAppliedLinearAccClamp=<500 500 500>
    componentLinearAccScale=<0.3 0.3 0.3>
    componentLinearVelScale=<0.3 0.3 0.3>
    gravityOverride {z=-980}
    physicsBodyDefinitions [{
      boundBone {boneName=Dyn_leaves_06}
      boxExtents=<10 10 10>
      constraintSetup {angularLimitsMax=<0 0 5>, angularLimitsMin=<0 -13 -5>, angularTarget=<1 0 0>}
      localJointOffset=<0.2 0 0>
     }]
   }
   n14 = Component To Local(ComponentPose=@n13)
   n15 = "Transform (Modify) Bone - Bone: Dyn_leaves_01"(
    ComponentPose=Local To Component(LocalPose=Blend Poses by bool(BlendPose_0=@n14, BlendPose_1=@n7))
    Alpha=1
    Translation="0,0,0"
    Rotation="2.512,2.629,-5.367"
    Scale="1,1,1"
   ) {rotationMode=BMM_Additive, boneToModify {boneName=Dyn_leaves_01}, rotation=<rot 2.512 2.629 -5.367>}
   Output Pose(Result=Component To Local(ComponentPose=@n15)) {layerGroup=None}
   AnimGraphNode_SubInput_0 = Input Pose()
   Save cached pose 'Base'(Pose=@AnimGraphNode_SubInput_0) {cacheName=Base}
  ]}
 EventGraph {nodes [
   TryGetPawnOwner()
   Event BlueprintUpdateAnimation()
  ]}
]
variables []
